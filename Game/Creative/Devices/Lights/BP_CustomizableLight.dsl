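; Creative customizable light device that switches between design, on and off states and shows a configurable or team-based color.
; Minigame phase, team-gated timed interaction, device events and music sequencer hits feed an authority-side SetState, which broadcasts activated/deactivated events.
; RepNotifies drive CreativeColorComponent fades and snaps on clients; team color index replicates on the actor as a workaround for component replication.
; Also migrates legacy options and toggles base visibility and collision.
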(event EventOnWorldReady
  (|CallParentFunction)
  ;[1 1 1] Enable/Disable collision for the base if it is visible.  It needs to be done here and on initialize in order to replicate properly
  (CallFunction|UpdateBaseVisibilityandCollision))

(event EventOnHitByMusicSequencer
  (params (InstigatedBy :type Controller%20Object%20Reference))
  ;[0.25 0.5 0.25] Toggle on Sequencer
  (if (Variables|UserOptions|GetToggleOnSequencer)
    (CallFunction|ToggleLight InstigatedBy)))

(event EventInitializeDevice
  (bind _adjustedcoloroption (Variables|Transient|GetAdjustedColorOption))
  (bind _colorcomponent (Variables|Default|GetColorComponent))
  ;[0.25 0.25 0.5] Internal
  (|CallParentFunction)
  (CallFunction|UpdateBaseVisibilityandCollision)
  (Default|AssignUpdateVisualsonClient (Variables|Default|GetColorComponent) (EventDispatchers|CreateEvent :!event Custom|UpdateLightVisualsOnClient))
  (CallFunction|SetColorComponentInitialValues (Variables|Default|GetFortMinigameProgress))
  ;[1 1 1] Setting initial values resets the color on the device, so if it has been changed via a team color change, we need to set it
  (if (Utilities|Enum|Equal(Enum) _adjustedcoloroption "NewEnumerator1")
    (Variables|/Game/Creative/BluePrints/ColorComponet/CreativeColorComponent.CreativeColorComponent_C|SetColorMapKeyOption _adjustedcoloroption _colorcomponent)
    (Class|CreativeColorComponent|SetTeamColorIndex _colorcomponent (Variables|Transient|GetTeamColorIndex) :!class /Game/Creative/BluePrints/ColorComponet/CreativeColorComponent.CreativeColorComponent_C)))

(event MinigameStateChangedDelegate(FortMinigameProgress)
  (params (Minigame :type Fort%20Minigame%20Object%20Reference) (NewMinigameState :type EFortMinigameState%20Enum :default "PreGame"))
  ;[0.25 0.5 0.25] Enabled on Phase
  (CallFunction|UpdateState 0))

(event Custom|UpdateLightVisualsOnClient)

(event Custom|ClientOnActivated
  ;[0.25 0.5 0.25] Colour Change Time
  (Class|CreativeColorComponent|FadetoBrightnessOptionValue (Variables|Default|GetColorComponent)))

(event Interaction|EventBlueprintOnInteract
  (params (InteractingPawn :type Fort%20Pawn%20Object%20Reference) (InteractionBeingAttempted :type EInteractionBeingAttempted%20Enum :default "FirstInteraction"))
  ;[0.25 0.25 0.5] Internal
  (CallFunction|HandleonInteract InteractingPawn))

(event Custom|ClientOnDeactivated
  ;[0.25 0.5 0.25] Colour Change Time
  (Class|CreativeColorComponent|SnaptoMinBrightness (Variables|Default|GetColorComponent)))

(event Custom|ClientOnDesign
  ;[0.25 0.5 0.25] Colour Change Time
  (Class|CreativeColorComponent|SnaptoBrightnessOptionValue (Variables|Default|GetColorComponent)))

(event Custom|ActivateEvent
  (params (Instigator :type Controller%20Object%20Reference))
  ;[0.5 0.25 0.25] On Received Enable
  (switch Utilities|FlowControl|Switch|SwitchonE_CustomizableLight_State (Variables|Transient|GetState)
    (:NewEnumerator2
      (CallFunction|SetState "NewEnumerator1" Instigator))))

(event MinigameEndedDelegate(FortMinigameLogic)
  ;[0.25 0.5 0.25] Minigame Ended
  (CallFunction|ResettoInitialValues 0))

(event Custom|DeactivateEvent
  (params (Instigator :type Controller%20Object%20Reference))
  ;[0.5 0.25 0.25] On Received Disable
  (switch Utilities|FlowControl|Switch|SwitchonE_CustomizableLight_State (Variables|Transient|GetState)
    (:NewEnumerator1
      (CallFunction|SetState "NewEnumerator2" Instigator))))

(event Custom|ToggleEvent
  (params (Instigator :type Controller%20Object%20Reference))
  ;[0.5 0.25 0.25] On Received Toggle
  (CallFunction|ToggleLight Instigator))

(event Custom|ResetEvent
  (params (Instigator :type Controller%20Object%20Reference))
  ;[0.5 0.25 0.25] On Received Reset
  (CallFunction|ResettoInitialValues Instigator))

(event Custom|SetToTeamColorEvent
  (params (Instigator :type Controller%20Object%20Reference))
  (bind (_teamindex _returnvalue) (Team|GetActorPvPTeamIndex Instigator))
  ;[0.5 0.25 0.25] On Received Set to Team Color
  (Utilities|IsValid Instigator
    (:Is%20Valid
      (if _returnvalue
        (Variables|Transient|SetAdjustedColorOption "NewEnumerator1")
        ;[1 1 1] This should be in the color component but the variables in actor componets don't seem to replication event with the variable RepNotify and the actor component with ComponentReplicates = true
        (Variables|Transient|SetTeamColorIndex (Knobs|Team|ReshiftTeamIndexValue (Math|Conversions|ToInteger(Byte) _teamindex)))))))

(fn BlueprintCanInteract
  (params (InteractingPawn :type Fort%20Pawn%20Object%20Reference) (InteractionBeingAttempted :type EInteractionBeingAttempted%20Enum :default "FirstInteraction") (InteractionType :type TInteractionType%20Enum :default "IT_NoInteraction"))
  (bind _interactaction (CallFunction|GetInteractionAction InteractingPawn))
  (switch Utilities|FlowControl|Switch|SwitchonE_CustomizableLight_InteractAction _interactaction
    (:NewEnumerator0
      (return false))
    (:NewEnumerator1
      (return (Interaction|CanEditOptionsComponent (Variables|Default|GetToyOptionsComponent) InteractingPawn)))
    (:NewEnumerator2
      _)
    (:NewEnumerator3
      _))
  (return true))

(fn BlueprintGetInteractionString
  (params (InteractingPawn :type Fort%20Pawn%20Object%20Reference) (InteractionBeingAttempted :type EInteractionBeingAttempted%20Enum :default "FirstInteraction"))
  (bind _interactaction (CallFunction|GetInteractionAction InteractingPawn))
  (return (Utilities|Select :Index _interactaction :NewEnumerator1 "NSLOCTEXT(\"[35BDD0AD432B00C074DFA082D9811BDD]\", \"F73F135E4446B4BD2EE9DCBAF94CE908\", \"Customize\")" :NewEnumerator2 "NSLOCTEXT(\"[35BDD0AD432B00C074DFA082D9811BDD]\", \"5466F3DB42B0881C55744ABBBEACFC0B\", \"Turn On\")" :NewEnumerator3 "NSLOCTEXT(\"[35BDD0AD432B00C074DFA082D9811BDD]\", \"8FCB2E42454B350C91126FB128A12D62\", \"Turn Off\")")))

(fn BlueprintGetInteractionTime
  (params (InteractingPawn :type Fort%20Pawn%20Object%20Reference) (InteractionBeingAttempted :type EInteractionBeingAttempted%20Enum :default "FirstInteraction"))
  (bind _interactaction (CallFunction|GetInteractionAction InteractingPawn))
  (return (Utilities|Select :Index _interactaction :NewEnumerator0 0.0 :NewEnumerator1 0.0 :NewEnumerator2 (Variables|UserOptions|GetTurnOnInteractTime) :NewEnumerator3 (Variables|UserOptions|GetTurnOffInteractTime)) true))

(fn GetInteractionAction
  (params (InteractingPawn :type Pawn%20Object%20Reference))
  (Utilities|FlowControl|Sequence
    (:then_0
      ;[1 1 1] Set Interacting Team if possible, otherwise leave as None
      (bind _interactingpawn (Variables|GetInteractingPawn)
        (:then
          (if (.ReturnValue (Team|GetActorPvPTeamIndex _interactingpawn))
            (LocalVariables|SetInteractingTeam (Knobs|Team|ReshiftTeamIndexValue (Math|Conversions|ToInteger(Byte) (Team|GetActorPvPTeamIndex _interactingpawn))))))
        (:else)))
    (:then_1
      (switch Utilities|FlowControl|Switch|SwitchonE_CustomizableLight_State (Variables|Transient|GetState)
        (:NewEnumerator0
          (LocalVariables|SetAction "NewEnumerator1"))
        (:NewEnumerator1
          ;[0 0 1] Is the player allowed to deactivate this light?
          (if (Math|Boolean|ORBoolean (Knobs|Team|IsTeamNone (LocalVariables|GetInteractingTeam)) (Creative|TeamOption|IsTeamType (Variables|UserOptions|GetTurnOffTeam) "All") (== (LocalVariables|GetInteractingTeam) (Creative|TeamOption|GetLegacyTeamIndex (Variables|UserOptions|GetTurnOffTeam))))
            (LocalVariables|SetAction "NewEnumerator3")))
        (:NewEnumerator2
          ;[0 1 0.06728] Is the pawn allowed to activate this light?
          (if (Math|Boolean|ORBoolean (Knobs|Team|IsTeamNone (LocalVariables|GetInteractingTeam)) (Creative|TeamOption|IsTeamType (Variables|UserOptions|GetTurnOnTeam) "All") (== (LocalVariables|GetInteractingTeam) (Creative|TeamOption|GetLegacyTeamIndex (Variables|UserOptions|GetTurnOnTeam))))
            (LocalVariables|SetAction "NewEnumerator2")))))
    (:then_2
      (return (LocalVariables|GetAction)))))

(fn GetMusicSequencerHitSound
  (params (InstigatedBy :type Controller%20Object%20Reference))
  (return 0 0 0 0))

(fn GetMusicSequencerHitSoundLocation
  (params (InstigatedBy :type Controller%20Object%20Reference))
  (return "0, 0, 0"))

(fn GetVisibilityComponents
  (return 0))

(fn HandleOnInteract
  (params (InteractingPawn :type Pawn%20Object%20Reference))
  (bind _interactingpawn (Variables|GetInteractingPawn))
  (bind _interactaction (CallFunction|GetInteractionAction InteractingPawn))
  (if (Utilities|Enum|NotEqual(Enum) _interactaction "NewEnumerator0")
    (Utilities|IsValid _interactingpawn
      (:Is%20Valid
        (LocalVariables|SetPawnController (Pawn|GetController _interactingpawn)))
      (:Is%20Not%20Valid
        _))
    (CallFunction|SetState (Utilities|Select :Index _interactaction :NewEnumerator0 "NewEnumerator0" :NewEnumerator1 "NewEnumerator0" :NewEnumerator2 "NewEnumerator1" :NewEnumerator3 "NewEnumerator2") (LocalVariables|GetPawnController))))

(fn IsActivatedAtCurrentPhase
  (bind _currentminigame (Variables|/Script/FortniteGame.FortMinigameLogicComponent|GetCurrentMinigame (Variables|Default|GetFortMinigameLogic))
    (:then
      (if (== (Variables|ToyOptions|GetActivateOnGamePhase) 0)
        (return false)
        (elif (Utilities|Enum|Equal(Enum) (Variables|/Script/FortniteGame.FortMinigame|GetCurrentState _currentminigame) "PreGame")
          (return (Utilities|Enum|Equal(Enum) (Variables|UserOptions|GetActivateOnPhase) "NewEnumerator1"))
          (elif (Utilities|Enum|Equal(Enum) (Variables|/Script/FortniteGame.FortMinigame|GetCurrentState _currentminigame) "Warmup")
            (return (Utilities|Enum|Equal(Enum) (Variables|UserOptions|GetActivateOnPhase) "NewEnumerator2"))
            (elif (Utilities|Enum|Equal(Enum) (Variables|/Script/FortniteGame.FortMinigame|GetCurrentState _currentminigame) "InProgress")
              (return (Utilities|Enum|Equal(Enum) (Variables|UserOptions|GetActivateOnPhase) "NewEnumerator3")))))))
    (:else
      _))
  (return false))

(fn IsColorVisible
  (params (AdjustedColor :type F%20HSVA%20Structure))
  (bind (_adjustedcolor_hue_2_285c8ddb4fae69cbf132f4b895363bd4 _adjustedcolor_saturation_4_96689e7d4094d69c09c353901b909cb7 _adjustedcolor_brightness_6_0a16ad3f4c6c520b30f6959130f47e90) (Utilities|Struct|BreakFHSVA (Variables|GetAdjustedColor_Hue_2_285C8DDB4FAE69CBF132F4B895363BD4)))
  (switch Utilities|FlowControl|Switch|SwitchonE_CustomizableLight_State (Variables|Transient|GetState)
    (:NewEnumerator0
      (return (or (Variables|UserOptions|Basic|GetInitialState) (CallFunction|IsActivatedatCurrentPhase))))
    (:NewEnumerator1
      (return (> _adjustedcolor_brightness_6_0a16ad3f4c6c520b30f6959130f47e90 0.0)))
    (:NewEnumerator2
      (return false))))

(fn MigrateLegacyUserOptions
  (bind _initialstateoption (Variables|ToyOptions|GetInitialStateOption))
  (bind _colorchangetimeoption (Variables|ToyOptions|GetColorChangeTimeOption))
  (bind _activateinteracttime (Variables|ToyOptions|GetActivateInteractTime))
  (bind _deactivateinteracttime (Variables|ToyOptions|GetDeactivateInteractTime))
  (Utilities|FlowControl|Sequence
    (:then_0
      ;[1 1 1] Migrate color change time to new override variables
      (if (not (Math|Float|NearlyEqual(Float) (Class|GetClassDefaults "/Game/Creative/Devices/Lights/BP_CustomizableLight.BP_CustomizableLight_C") _colorchangetimeoption))
        (Variables|UserOptions|SetColorChangeTime_Override true)
        (Variables|UserOptions|SetColorChangeTime _colorchangetimeoption)
        (Variables|ToyOptions|SetColorChangeTimeOption (Class|GetClassDefaults "/Game/Creative/Devices/Lights/BP_CustomizableLight.BP_CustomizableLight_C"))
        (LocalVariables|SetOptions%20Migrated true)))
    (:then_1
      ;[1 1 1] Migrate activate interact time (int) to Turn On time (float with override)
      (if (!= (Class|GetClassDefaults "/Game/Creative/Devices/Lights/BP_CustomizableLight.BP_CustomizableLight_C") _activateinteracttime)
        (Variables|UserOptions|SetTurnOnInteractTime_Override true)
        (Variables|UserOptions|SetTurnOnInteractTime (Math|Conversions|ToFloat(Integer) _activateinteracttime))
        (Variables|ToyOptions|SetActivateInteractTime (Class|GetClassDefaults "/Game/Creative/Devices/Lights/BP_CustomizableLight.BP_CustomizableLight_C"))
        (LocalVariables|SetOptions%20Migrated true)))
    (:then_2
      ;[1 1 1] Migrate deactivate interact time (int) to Turn Off time (float with override)
      (if (!= (Class|GetClassDefaults "/Game/Creative/Devices/Lights/BP_CustomizableLight.BP_CustomizableLight_C") _deactivateinteracttime)
        (Variables|UserOptions|SetTurnOffInteractTime_Override true)
        (Variables|UserOptions|SetTurnOffInteractTime (Math|Conversions|ToFloat(Integer) _deactivateinteracttime))
        (Variables|ToyOptions|SetDeactivateInteractTime (Class|GetClassDefaults "/Game/Creative/Devices/Lights/BP_CustomizableLight.BP_CustomizableLight_C"))
        (LocalVariables|SetOptions%20Migrated true)))
    (:then_3
      ;[1 1 1] Migrate initial state from Int to Bool. (Doing this here because the redirect was importing incorrectly.)
      (if (!= (Class|GetClassDefaults "/Game/Creative/Devices/Lights/BP_CustomizableLight.BP_CustomizableLight_C") _initialstateoption)
        (Variables|UserOptions|Basic|SetInitialState (Math|Conversions|ToBoolean(Integer) _initialstateoption))
        (Variables|ToyOptions|SetInitialStateOption (Class|GetClassDefaults "/Game/Creative/Devices/Lights/BP_CustomizableLight.BP_CustomizableLight_C"))
        (LocalVariables|SetOptions%20Migrated true)))
    (:then_4
      ;[1 1 1] Refresh overrides if needed
      (if (LocalVariables|GetOptions%20Migrated)
        (OptionData|RefreshPropertyOverrides (Variables|Default|GetToyOptionsComponent))))))

(fn OnRep_State
  (Networking|CanExecuteCosmeticEvents
    (:True
      (Utilities|FlowControl|Sequence
        (:then_0
          (switch Utilities|FlowControl|Switch|SwitchonE_CustomizableLight_State (Variables|Transient|GetState)
            (:NewEnumerator0
              (CallFunction|ClientonDesign))
            (:NewEnumerator1
              (CallFunction|ClientonActivated))
            (:NewEnumerator2
              (CallFunction|ClientonDeactivated))))
        (:then_1
          (CallFunction|UpdateLightVisualsonClient))))))

(fn OnRep_TeamColorIndex
  (bind _colorcomponent (Variables|Default|GetColorComponent))
  (Networking|CanExecuteCosmeticEvents
    (:True
      (Variables|/Game/Creative/BluePrints/ColorComponet/CreativeColorComponent.CreativeColorComponent_C|SetColorMapKeyOption (Variables|Transient|GetAdjustedColorOption) _colorcomponent)
      (Class|CreativeColorComponent|SetTeamColorIndex _colorcomponent (Variables|Transient|GetTeamColorIndex) :!class /Game/Creative/BluePrints/ColorComponet/CreativeColorComponent.CreativeColorComponent_C))))

(fn ResetToInitialValues
  (params (StateInstigator :type Controller%20Object%20Reference))
  ;[0.5 0.25 0.25] Colors
  (Class|CreativeColorComponent|ResettoInitalValues (Variables|Default|GetColorComponent))
  (Networking|SwitchHasAuthority
    (:Authority
      ;[0.25 0.25 0.25] State
      (CallFunction|UpdateState (Variables|GetStateInstigator))
      ;[1 1 1] Team Color/Default Color Setting
      (Variables|Transient|SetAdjustedColorOption (Variables|UserOptions|Basic|GetColor))
      ;[1 1 1] TeamColor
      (Variables|Transient|SetTeamColorIndex (Creative|TeamOption|GetLegacyTeamIndex (Variables|UserOptions|GetInitialTeamColor) false)))))

(fn SetColorComponentInitialValues
  (params (MinigameLogic :type Fort%20Minigame%20Logic%20Component%20Object%20Reference%20%28by%20ref%29))
  (Class|CreativeColorComponent|SetInitialValues (Variables|Default|GetColorComponent) (Creative|TeamOption|GetLegacyTeamIndex (Variables|UserOptions|GetInitialTeamColor) false) (Variables|UserOptions|Basic|GetColor) (Math|Color|MakeColor (Variables|ToyOptions|GetCustomRedOption) (Variables|ToyOptions|GetCustomGreenOption) (Variables|ToyOptions|GetCustomBlueOption)) (Variables|UserOptions|GetBrightnessScale) (Variables|UserOptions|GetColorChangeTime) (Variables|GetMinigameLogic)))

(fn SetState
  (params (NewState :type E_CustomizableLight_State%20Enum :default "NewEnumerator0") (StateInstigator :type Controller%20Object%20Reference))
  (Networking|SwitchHasAuthority
    (:Authority
      (if (Utilities|Enum|NotEqual(Enum) NewState (Variables|Transient|GetState))
        ;[1 1 1] If going to Design state, ensure the mini game is not in progress
        (if (Utilities|Enum|Equal(Enum) (Variables|GetNewState) "NewEnumerator0")
          (bind _currentminigame (Variables|/Script/FortniteGame.FortMinigameLogicComponent|GetCurrentMinigame (Variables|Default|GetFortMinigameLogic))
            (:then
              (Minigame|IsInProgress _currentminigame
                (:No
                  _)))
            (:else))
          (else
            _))
        (bind _output_get (Variables|Transient|SetState (Variables|GetNewState)))
        (switch Utilities|FlowControl|Switch|SwitchonE_CustomizableLight_State _output_get
          (:NewEnumerator1
            (GameplayEvent|BroadcastEvent (Variables|UserOptions|Events|GetOn%20Activated) 0 (Utilities|Struct|MakeGameplayEventActivated (Variables|GetStateInstigator))))
          (:NewEnumerator2
            (GameplayEvent|BroadcastEvent (Variables|UserOptions|Events|GetOn%20Deactivated) 0 (Utilities|Struct|MakeGameplayEventDeactivated (Variables|GetStateInstigator)))))))))

(fn ToggleLight
  (params (StateInstigator :type Controller%20Object%20Reference))
  (bind _stateinstigator (Variables|GetStateInstigator))
  (switch Utilities|FlowControl|Switch|SwitchonE_CustomizableLight_State (Variables|Transient|GetState)
    (:NewEnumerator1
      (CallFunction|SetState "NewEnumerator2" _stateinstigator))
    (:NewEnumerator2
      (CallFunction|SetState "NewEnumerator1" _stateinstigator))))

(fn UpdateBaseVisibilityAndCollision
  (bind _basevisible (Variables|UserOptions|GetBaseVisible))
  (bind _returnvalue (select _basevisible "ECR_Block" "ECR_Ignore"))
  (bind _staticmeshcomponent (Variables|Debug|GetStaticMeshComponent))
  (Collision|SetCollisionResponsetoChannel _staticmeshcomponent "ECC_Pawn" _returnvalue)
  (Collision|SetCollisionResponsetoChannel _staticmeshcomponent "ECC_Camera" _returnvalue)
  (UEFNConversionHelper|SetComponentHiddenEditor (Variables|Debug|GetStaticMeshComponent) (not (Variables|UserOptions|GetBaseVisible))))

(fn UpdateState
  (params (StateInstigator :type Controller%20Object%20Reference))
  (bind _stateinstigator (Variables|GetStateInstigator))
  (bind _currentminigame (Variables|/Script/FortniteGame.FortMinigameLogicComponent|GetCurrentMinigame (Variables|Default|GetFortMinigameLogic))
    (:then
      (if (Utilities|Enum|Equal(Enum) (Variables|/Script/FortniteGame.FortMinigame|GetCurrentState _currentminigame) "PreGame")
        (CallFunction|SetState "NewEnumerator0" _stateinstigator)
        (elif (CallFunction|IsActivatedatCurrentPhase)
          (CallFunction|SetState "NewEnumerator1" _stateinstigator)
          (else
            (CallFunction|SetState (select (Variables|UserOptions|Basic|GetInitialState) "NewEnumerator1" "NewEnumerator2") (Variables|GetStateInstigator))))))
    (:else)))

(fn ConstructionScript
  (|CallParentFunction)
  (CallFunction|ResettoInitialValues 0)
  ;[1 1 1] So colors update properly in UEFN
  (if (Editor|InEditor)
    (CallFunction|UpdateLightVisualsonClient)))
